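; Gameplay ability for a Season 24 Battle Pass quest about entering a vehicle after landing from the bus skydive.
; On event activation it calls the parent, then waits once for the Event.SkydiveEnded.FromBus tag to be removed.
; On removal it applies the quest's after-landing timer gameplay effect to the owner and ends itself.
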(event Ability|EventActivateAbilityFromEvent
  (params (EventData :type Gameplay%20Event%20Data%20Structure%20%28by%20ref%29))
  (|CallParentFunction EventData)
  (Ability|Tasks|WaitGameplayTagRemove "(TagName=\"Event.SkydiveEnded.FromBus\")" 0 true
    (:Removed
      (Ability|ApplyGameplayEffectToOwner "/BattlePassS24/Items/QuestAbilities/Quest_EnterVehicleAfterLanding/GE_Quest_S24_Enter_After_Landing_Timer.GE_Quest_S24_Enter_After_Landing_Timer_C")
      (Ability|EndAbility))))
